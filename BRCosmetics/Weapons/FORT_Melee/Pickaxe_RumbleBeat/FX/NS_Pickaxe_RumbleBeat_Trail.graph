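# Niagara ribbon trail effect for swings of the RumbleBeat pickaxe, built on anim-trail modules that follow bone or socket pairs.
# Each emitter runs four trails, with User.SwingSectionIndex selecting socket pairs from per-trail integer arrays.
# HammerTrail is a short-lived dark, time-of-day tinted swoosh.
# LightningTrail_Custom spawns only when User.HasHitThing is true, adding bright cyan ribbons with separate glow and swoosh colors faded by age and time of day.
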NiagaraSystem /BRCosmetics/Weapons/FORT_Melee/Pickaxe_RumbleBeat/FX/NS_Pickaxe_RumbleBeat_Trail
dependencies {
 usedDataInterfaces [
  &NiagaraDataInterfaceArrayInt32
  &NiagaraDataInterfaceCurve
  &NiagaraDataInterfaceColorCurve
  &NiagaraDataInterfaceArrayFloat
  &NiagaraDataInterfaceSocketReader
 ]
 usedDynamicInputs []
 usedModules [
  &self:SystemSpawnScript.SystemScriptSource.SystemScriptGraph.NiagaraNodeAssignment_2.SetVariables_#1
  &self:SystemSpawnScript.SystemScriptSource.SystemScriptGraph.NiagaraNodeAssignment_1.SetVariables_#2
  &SystemState
 ]
}
emitters {
 HammerTrail {
  emitterSpawnScript [&FN_RibbonIDOffsetUtility()]
  emitterUpdateScript [
   &EmitterState()
   &FN_SpawnAnimTrail(Number of Anim Trails=4)
  ]
  particleSpawnScript [
   &FN_InitializeAnimTrailBoneIndecies(
    First Bone or Socket Index:NiagaraInt32=&SelectIntFromArray(Direct Array Index=$User.SwingSectionIndex, Int Selection Array=DataInterface() {propertyValues {IntData [6,1,5,2]}})
    Second Bone or Socket Index:NiagaraInt32=&SelectIntFromArray(Direct Array Index=$User.SwingSectionIndex, Int Selection Array=DataInterface() {propertyValues {IntData [2,5,1,6]}})
   )
   &FN_InitializeAnimTrailBoneIndecies(
    AnimTrail Selection=1
    First Bone or Socket Index:NiagaraInt32=&SelectIntFromArray(Direct Array Index=$User.SwingSectionIndex, Int Selection Array=DataInterface() {propertyValues {IntData [3,0,4,7]}})
    Second Bone or Socket Index:NiagaraInt32=&SelectIntFromArray(Direct Array Index=$User.SwingSectionIndex, Int Selection Array=DataInterface() {propertyValues {IntData [7,4,0,3]}})
   ) {module=FN_InitializeAnimTrailBoneIndecies001}
   &FN_InitializeAnimTrailBoneIndecies(
    AnimTrail Selection=2
    First Bone or Socket Index:NiagaraInt32=&SelectIntFromArray(Direct Array Index=$User.SwingSectionIndex, Int Selection Array=DataInterface() {propertyValues {IntData [3,0,0,7]}})
    Second Bone or Socket Index:NiagaraInt32=&SelectIntFromArray(Direct Array Index=$User.SwingSectionIndex, Int Selection Array=DataInterface() {propertyValues {IntData [2,1,1,6]}})
   ) {module=FN_InitializeAnimTrailBoneIndecies002}
   &FN_InitializeAnimTrailBoneIndecies(
    AnimTrail Selection=3
    First Bone or Socket Index:NiagaraInt32=&SelectIntFromArray(Direct Array Index=$User.SwingSectionIndex, Int Selection Array=DataInterface() {propertyValues {IntData [6,5,5,2]}})
    Second Bone or Socket Index:NiagaraInt32=&SelectIntFromArray(Direct Array Index=$User.SwingSectionIndex, Int Selection Array=DataInterface() {propertyValues {IntData [7,4,4,3]}})
   ) {module=FN_InitializeAnimTrailBoneIndecies003}
   &FN_InitializeAnimTrail(Lifetime=0.2)
   &AddVelocity(Velocity=<-402 0 200>) {enabled=false}
  ]
  particleUpdateScript [
   &FN_UpdateAnimTrail() {module=FN_UpdateAnimTrail001}
   /SetVariables(
    Particles.AlphaByEmitterAge:NiagaraFloat=&FloatFromCurve(
     FloatCurve=DataInterface() {propertyValues {Curve {keys [
         {interpMode=RCIM_Cubic, leaveTangentWeight=0.2033, tangentMode=RCTM_User, time=0.1, value=1}
         {arriveTangent=-11.995, tangentMode=RCTM_User, time=0.3, value=0}
        ]}}}
     CurveIndex=$Emitter.Age
    )
    Particles.AlphaByNormalizedAge:NiagaraFloat=&FloatFromCurve(FloatCurve=DataInterface() {propertyValues {Curve {keys [{interpMode=RCIM_Cubic, tangentMode=RCTM_User, time=0.2, value=1}, {interpMode=RCIM_Cubic, tangentMode=RCTM_User, time=1, value=0}]}}})
   ) {isSetParameters=true, module=SetVariables_#3}
   &SolveForcesAndVelocity(Clamp Velocity=false, Limit Acceleration=false)
   &ScaleColor(
    Scale Alpha=&FloatFromCurve(FloatCurve=DataInterface() {propertyValues {Curve {keys [<at 0 1 RCIM_Cubic>, {arriveTangent=-3.0434, interpMode=RCIM_Cubic, leaveTangent=-3.0434, tangentMode=RCTM_User, time=1, value=0}]}}})
    ScaleA=true
    ScaleRGB=true
   )
   &ScaleColor(
    Scale Alpha=&FNTimeOfDayScalar(Night Scale=0.8, Morning Scale=0.75, Day Scale=0.8, Evening Scale=0.75)
    Scale RGB=&MakeVectorFromLinearColorRGB(Color=&FNTimeOfDayColor(
      Night Color=<rgba 0.0122 0.0193 0.0469 1>
      Morning Color=<rgba 0.0108 0.013 0.0208 1>
      Day Color=<rgba 0.0093 0.0232 0.0469 1>
      Evening Color=<rgba 0.0208 0.0208 0.0208 1>
     ))
    ScaleA=true
    ScaleRGB=true
   ) {module=ScaleColor001}
  ]
  rendererClasses [&NiagaraRibbonRendererProperties]
  renderers [
   {rendererClass=&NiagaraRibbonRendererProperties}
   {rendererClass=&NiagaraRibbonRendererProperties, rendererIndex=1}
  ]
 }
 LightningTrail_Custom {
  emitterSpawnScript [&FN_RibbonIDOffsetUtility()]
  emitterUpdateScript [
   &EmitterState()
   &FN_SpawnAnimTrail(Number of Anim Trails:NiagaraInt32=&MakeCustomIntFromBool(Bool=$User.HasHitThing, True INT=4, False INT=0))
  ]
  particleSpawnScript [
   &FN_InitializeAnimTrailBoneIndecies(
    First Bone or Socket Index:NiagaraInt32=&SelectIntFromArray(Direct Array Index=$User.SwingSectionIndex, Int Selection Array=DataInterface() {propertyValues {IntData [6,6,2,6]}})
    Second Bone or Socket Index:NiagaraInt32=&SelectIntFromArray(Direct Array Index=$User.SwingSectionIndex, Int Selection Array=DataInterface() {propertyValues {IntData [2,2,6,2]}})
   )
   &FN_InitializeAnimTrailBoneIndecies(
    AnimTrail Selection=1
    First Bone or Socket Index:NiagaraInt32=&SelectIntFromArray(Direct Array Index=$User.SwingSectionIndex, Int Selection Array=DataInterface() {propertyValues {IntData [7,7,7,7]}})
    Second Bone or Socket Index:NiagaraInt32=&SelectIntFromArray(Direct Array Index=$User.SwingSectionIndex, Int Selection Array=DataInterface() {propertyValues {IntData [3,3,3,3]}})
   ) {module=FN_InitializeAnimTrailBoneIndecies001}
   &FN_InitializeAnimTrailBoneIndecies(
    AnimTrail Selection=2
    First Bone or Socket Index:NiagaraInt32=&SelectIntFromArray(Direct Array Index=$User.SwingSectionIndex, Int Selection Array=DataInterface() {propertyValues {IntData [3,3,3,3]}})
    Second Bone or Socket Index:NiagaraInt32=&SelectIntFromArray(Direct Array Index=$User.SwingSectionIndex, Int Selection Array=DataInterface() {propertyValues {IntData [2,2,2,2]}})
   ) {module=FN_InitializeAnimTrailBoneIndecies002}
   &FN_InitializeAnimTrailBoneIndecies(
    AnimTrail Selection=3
    First Bone or Socket Index:NiagaraInt32=&SelectIntFromArray(Direct Array Index=$User.SwingSectionIndex, Int Selection Array=DataInterface() {propertyValues {IntData [6,6,6,6]}})
    Second Bone or Socket Index:NiagaraInt32=&SelectIntFromArray(Direct Array Index=$User.SwingSectionIndex, Int Selection Array=DataInterface() {propertyValues {IntData [7,7,7,7]}})
   ) {module=FN_InitializeAnimTrailBoneIndecies003}
   &FN_InitializeAnimTrail(Color=<rgba 1.4454 16.6118 20 1>, Lifetime=0.2)
   &AddVelocity(Velocity=<400 0 0>) {enabled=false}
  ]
  particleUpdateScript [
   &FN_UpdateAnimTrail() {module=FN_UpdateAnimTrail001}
   /SetVariables(
    Particles.AlphaByEmitterAge:NiagaraFloat=&FloatFromCurve(
     FloatCurve=DataInterface() {propertyValues {Curve {keys [
         {interpMode=RCIM_Cubic, leaveTangentWeight=0.2033, tangentMode=RCTM_User, time=0.1, value=1}
         {arriveTangent=-11.995, tangentMode=RCTM_User, time=0.3, value=0}
        ]}}}
     CurveIndex=$Emitter.Age
    )
    Particles.AlphaByNormalizedAge:NiagaraFloat=&FloatFromCurve(FloatCurve=DataInterface() {propertyValues {Curve {keys [
         {interpMode=RCIM_Cubic, tangentMode=RCTM_User, time=0, value=1}
         {arriveTangent=-6.1356, interpMode=RCIM_Cubic, leaveTangent=-6.1356, tangentMode=RCTM_User, time=0.3455, value=0}
        ]}}})
   ) {isSetParameters=true, module=SetVariables_#3}
   &ScaleColor(
    Scale Alpha=&Multiply_Float(A=1, B=$Particles.AlphaByNormalizedAge)
    Scale RGB=&MakeVectorFromLinearColorRGB(Color=&ColorFromCurve(ColorCurve=DataInterface() {propertyValues {
        AlphaCurve {keys []}
        BlueCurve {keys [<at 0 1 RCIM_Cubic>,<at 0.6466 1 RCIM_Cubic>]}
        GreenCurve {keys [<at 0 1 RCIM_Cubic>,<at 0.6466 0.8306 RCIM_Cubic>]}
        RedCurve {keys [<at 0 1 RCIM_Cubic>,<at 0.6466 0.0723 RCIM_Cubic>]}
       }}))
    ScaleA=true
    ScaleRGB=true
   ) {module=ScaleColor002}
   /SetVariables(Particles.Glow Color=&MakeLinearColorFromVectorAndFloat(
     "Vector (RGB)"=&MakeVectorFromLinearColorRGB(Color=&MultiplyLinearColor(
       Linear Color=<rgba 0.1445 1.6612 2 1>
       Scale Factor=&MakeVector4_FromLinearColor(LinearColor=&ColorFromCurve(ColorCurve=DataInterface() {propertyValues {
           AlphaCurve {keys []}
           BlueCurve {keys [<at 0 1.5 RCIM_Cubic>,<at 0.749 2>]}
           GreenCurve {keys [<at 0 1.5 RCIM_Cubic>,<at 0.749 1.6612>]}
           RedCurve {keys [<at 0 1.5 RCIM_Cubic>,<at 0.749 0.1445>]}
          }}))
      ))
     "Float (Alpha)"=&Multiply_Float(A=1, B=$Particles.AlphaByNormalizedAge)
    )) {isSetParameters=true}
   /SetVariables(Particles.Swoosh Color:LinearColor=&MakeLinearColorFromVectorAndFloat(
     "Vector (RGB)"=&MakeVectorFromLinearColorRGB(Color=&MultiplyLinearColor(
       Linear Color=<rgba 0.0145 0.1 0.2 1>
       Scale Factor=&MakeVector4_FromLinearColor(LinearColor=&ColorFromCurve(ColorCurve=DataInterface() {propertyValues {
           AlphaCurve {keys []}
           BlueCurve {keys [<at 0 1.5 RCIM_Cubic>,<at 0.749 0.2>]}
           GreenCurve {keys [<at 0 1.5 RCIM_Cubic>,<at 0.749 0.1>]}
           RedCurve {keys [<at 0 1.5 RCIM_Cubic>,<at 0.749 0.0145>]}
          }}))
      ))
     "Float (Alpha)"=&Multiply_Float(A=1, B=$Particles.AlphaByNormalizedAge)
    )) {isSetParameters=true}
   &DynamicMaterialParameters(
    Index 0 Param 2=&FloatFromCurve(FloatCurve=DataInterface() {propertyValues {Curve {keys [
         {arriveTangent=-0.0204, interpMode=RCIM_Cubic, leaveTangent=-0.0204, tangentMode=RCTM_User, time=0.491, value=0}
         {arriveTangent=-2.2131, interpMode=RCIM_Cubic, leaveTangent=-2.2131, tangentMode=RCTM_User, time=1, value=-1}
        ]}}})
    Module.Index0Float0Write=true
    Module.Index0Float1Write=true
    Module.Index0Float2Write=true
    Module.Index0Float3Write=true
   ) {enabled=false}
   /SetVariables(Particles.Glow Dyn Mat Param:Vector4f=&MakeVector4(X=1, Y=1, Z=1, W=1)) {isSetParameters=true}
   &SolveForcesAndVelocity(Clamp Velocity=false, Limit Acceleration=false)
   /SetVariables(Particles.MainTrailWidth:NiagaraFloat=&Add_Float(
     A=$Particles.RibbonWidth
     B=&SelectFloatFromArray(Direct Array Index=$Particles.SpawnGroup, Float Selection Array=DataInterface() {propertyValues {FloatData [-20,-20,0,0]}})
    )) {isSetParameters=true}
   &ScaleColor(
    Scale RGB=&VectorFromFloat(Value=&FNTimeOfDayScalar(Night Scale=0.25, Morning Scale=1, Evening Scale=1))
    ScaleA=true
    ScaleRGB=true
   )
   /SetVariables(
    Particles.Glow Color=&Multiply_LinearColorByFloat(
     Linear Color=$Particles.Glow Color
     Float=&FNTimeOfDayScalar(Night Scale=0.75, Morning Scale=1, Evening Scale=1)
    )
    Particles.Swoosh Color:LinearColor=&Multiply_LinearColorByFloat(
     Linear Color=$Particles.Swoosh Color
     Float=&FNTimeOfDayScalar(Night Scale=1, Morning Scale=1, Evening Scale=1)
    )
   ) {isSetParameters=true}
  ]
  rendererClasses [&NiagaraRibbonRendererProperties]
  renderers [
   {rendererClass=&NiagaraRibbonRendererProperties}
   {rendererClass=&NiagaraRibbonRendererProperties, rendererIndex=1}
   {rendererClass=&NiagaraRibbonRendererProperties, rendererIndex=2}
   {rendererClass=&NiagaraRibbonRendererProperties, rendererIndex=3}
  ]
 }
}
summary {
 systemName=NS_Pickaxe_RumbleBeat_Trail
 emitters [
  {emitterName=LightningTrail_Custom, rendererClasses [&NiagaraRibbonRendererProperties]}
  {emitterName=HammerTrail, rendererClasses [&NiagaraRibbonRendererProperties]}
 ]
 userVariables [
  User.bIsFrontEnd {type=NiagaraBool}
  User.bIsFrontEndPreview {type=NiagaraBool}
  User.HasHitThing {type=NiagaraBool}
  User.RibbonIDOffset {type=NiagaraInt32}
  User.SwingSectionIndex {type=NiagaraInt32}
 ]
}
