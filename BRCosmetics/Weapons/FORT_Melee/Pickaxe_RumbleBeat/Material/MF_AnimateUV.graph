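# MF_AnimateUV is a material function in the RumbleBeat pickaxe cosmetic folder that outputs tiled, scrolling UVs as AnimatedUV.
# It takes UVs plus separate X and Y tile and speed multipliers, passed through color-coded named reroutes.
# A multiply by an appended two-component vector feeds a time-driven Panner, whose speed is another appended pair.
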MaterialFunction /BRCosmetics/Weapons/FORT_Melee/Pickaxe_RumbleBeat/Material/MF_AnimateUV
graph {nodes [
  n1 = NamedRerouteUsage()
  n2 = AppendVector(A=NamedRerouteUsage(), B=NamedRerouteUsage())
  n3 = NamedRerouteUsage()
  n4 = NamedRerouteUsage()
  FunctionOutput(None=Panner(Coordinate=Multiply(A=@n1, B=@n2), Time=Time(), Speed=AppendVector(A=@n3, B=@n4))) {outputName=AnimatedUV}
  NamedRerouteDeclaration(None=FunctionInput() {inputName=XTileMultiplier, inputType=FunctionInput_Scalar, sortPriority=1}) {name=XTileMultiplier, nodeColor=<rgba 1 0 0.4687 1>}
  NamedRerouteDeclaration(None=FunctionInput() {inputName=YTileMultiplier, inputType=FunctionInput_Scalar, sortPriority=2}) {name=YTileMultiplier, nodeColor=<rgba 0 1 0.5003 1>}
  NamedRerouteDeclaration(None=FunctionInput() {inputName=XSpeedMultiplier, inputType=FunctionInput_Scalar, sortPriority=3}) {name=XSpeedMultiplier, nodeColor=<rgba 0.9302 1 0.0308 1>}
  NamedRerouteDeclaration(None=FunctionInput() {inputName=YSpeedMultiplier, inputType=FunctionInput_Scalar, sortPriority=4}) {name=YSpeedMultiplier, nodeColor=<rgba 0 0.0107 1 1>}
  NamedRerouteDeclaration(None=FunctionInput() {inputName=UVs, inputType=FunctionInput_Vector2}) {name=UV, nodeColor=<rgba 0 1 0.25 1>}
 ]}
