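# Material function that adds animated, color-graded glow to the Rumble Beat pickaxe's emissive, passing other material attributes through.
# Packed noise channels pan in UV and local-position radial space, colored by three-color lerps, with a core texture and sphere and box masks.
# Radial samples take mip derivatives from mesh UVs, avoiding polar seams.
# An FX mask input and timestamp-driven up/down ramps, computed per vertex, support timed glow transitions.
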MaterialFunction /BRCosmetics/Weapons/FORT_Melee/Pickaxe_RumbleBeat/Material/MF_RumbleBeat_Pickaxe_Techart
graph {nodes [
  Constant() {r=0.5}
  n1 = GetMaterialAttributes(None=NamedRerouteUsage()) {attributeGetTypes [#1]}
  FunctionOutput(None=SetMaterialAttributes!(MaterialAttributes=@n1.MaterialAttributes, EmissiveColor=NamedRerouteUsage()) {attributeSetTypes [#1]})
  n2 = NamedRerouteUsage()
  n3 = NamedRerouteUsage()
  NamedRerouteDeclaration(None=Multiply(A=Multiply(A=@n2, B=@n3), B=NamedRerouteUsage())) {name=AnimatedColors, nodeColor=<rgba 1 0.7266 0 1>}
  n4 = TextureCoordinate()
  n5 = NamedRerouteUsage()
  n6 = NamedRerouteUsage()
  n7 = NamedRerouteUsage()
  n8 = NamedRerouteUsage()
  n9 = TextureObjectParameter() {
   parameterName=PackedNoise
   samplerType=SAMPLERTYPE_LinearColor
   texture=/BRCosmetics/Weapons/FORT_Melee/Pickaxe_RumbleBeat/Textures/RumbleBeatGlow_Noise
  }
  NamedRerouteDeclaration(None=TextureSample(
    UVs=MaterialFunctionCall!(UVs=@n4, XTileMultiplier=@n5, YTileMultiplier=@n6, XSpeedMultiplier=@n7, YSpeedMultiplier=@n8) {materialFunction=&/MF_AnimateUV}.AnimatedUV
    Tex=@n9
   ).R) {name=RedChannelTexture, nodeColor=<rgba 1 0 0 1>}
  NamedRerouteDeclaration(None=ScalarParameter() {defaultValue=0.16, group=RumbleBeat_TechArt, parameterName=GreenChannel_UVXSpeed, sortPriority=38}) {name=GreenChannel_UVXSpeed, nodeColor=<rgba 0 1 0.0008 1>}
  NamedRerouteDeclaration(None=ScalarParameter() {defaultValue=-0.01, group=RumbleBeat_TechArt, parameterName=GreenChannel__UVYSpeed, sortPriority=39}) {name=GreenChannel_UVYSpeed, nodeColor=<rgba 0 1 0.0008 1>}
  NamedRerouteDeclaration(None=ScalarParameter() {defaultValue=1, group=RumbleBeat_TechArt, parameterName=BlueChannel_UVXScale, sortPriority=40}) {name=BlueChannel_UVXScale, nodeColor=<rgba 0 0 1 1>}
  NamedRerouteDeclaration(None=ScalarParameter() {defaultValue=2, group=RumbleBeat_TechArt, parameterName=BlueChannel_UVYScale, sortPriority=41}) {name=BlueChannel_UVYScale, nodeColor=<rgba 0 0 1 1>}
  NamedRerouteDeclaration(None=ScalarParameter() {defaultValue=0.21, group=RumbleBeat_TechArt, parameterName=BlueChannel_UVXSpeed, sortPriority=42}) {name=BlueChannel_UVXSpeed, nodeColor=<rgba 0 0 1 1>}
  NamedRerouteDeclaration(None=ScalarParameter() {defaultValue=0.01, group=RumbleBeat_TechArt, parameterName=BlueChannel__UVYSpeed, sortPriority=43}) {name=BlueChannel_UVYSpeed, nodeColor=<rgba 0 0 1 1>}
  NamedRerouteDeclaration(None=ScalarParameter() {group=RumbleBeat_TechArt, parameterName=TimestampA}) {name=TimestampA, nodeColor=<rgba 0 0.3672 1 1>}
  NamedRerouteDeclaration(None=ScalarParameter() {group=RumbleBeat_TechArt, parameterName=TransitionLength_InA}) {name=TransitionLength_InA, nodeColor=<rgba 0.6875 1 0 1>}
  n10 = NamedRerouteUsage()
  n11 = NamedRerouteUsage()
  NamedRerouteDeclaration(None=VertexInterpolator(VS=MaterialFunctionCall(TimestampA=@n10, TransitionLength_In=@n11) {materialFunction=&MF_Timestamp_InOut_01}.Linear).PS) {name=PannerTimeB, desc=ramp down, nodeColor=<rgba 1 0.965 0.0428 1>}
  NamedRerouteDeclaration(None=ScalarParameter() {group=RumbleBeat_TechArt, parameterName=TimestampB}) {name=TimestampB, nodeColor=<rgba 1 0.0563 0.4827 1>}
  n12 = TextureCoordinate()
  n13 = NamedRerouteUsage()
  n14 = NamedRerouteUsage()
  n15 = NamedRerouteUsage()
  n16 = NamedRerouteUsage()
  n17 = TextureObjectParameter() {
   group=RumbleBeat_TechArt
   parameterName=PackedNoise
   samplerType=SAMPLERTYPE_LinearColor
   texture=&RumbleBeatGlow_Noise
  }
  NamedRerouteDeclaration(None=TextureSample(
    UVs=MaterialFunctionCall(UVs=@n12, XTileMultiplier=@n13, YTileMultiplier=@n14, XSpeedMultiplier=@n15, YSpeedMultiplier=@n16) {materialFunction=&MF_AnimateUV}.AnimatedUV
    Tex=@n17
   ).G) {name=GreenChannelTexture, nodeColor=<rgba 0 1 0.0859 1>}
  NamedRerouteDeclaration(None=ScalarParameter() {group=RumbleBeat_TechArt, parameterName=TransitionLength_InB}) {name=TransitionLength_InB, nodeColor=<rgba 0.0704 1 0.2285 1>}
  NamedRerouteDeclaration(None=ScalarParameter() {group=RumbleBeat_TechArt, parameterName=ToggleGlow}) {name=ToggleGlow, nodeColor=<rgba 0.0312 0 1 1>}
  n18 = TextureSample() {
   samplerType=SAMPLERTYPE_LinearColor
   texture=/BRCosmetics/Weapons/FORT_Melee/Pickaxe_RumbleBeat/Textures/T_RumbleBeat_Pickaxe_T_RumbleBeat_Pickaxe_FX
  }
  NamedRerouteDeclaration(None=FunctionInput(Preview=@n18.RGB) {inputName=In_TechArtMask, sortPriority=1}) {name=In_TechArtMask, nodeColor=<rgba 0.2891 1 0 1>}
  NamedRerouteDeclaration(None=FunctionInput() {inputName=In_MaterialAttributes, inputType=FunctionInput_MaterialAttributes}) {name=In_MaterialAttributes, nodeColor=<rgba 1 0 0.6094 1>}
  NamedRerouteDeclaration(None=VectorParameter() {group=RumbleBeat_TechArt, parameterName=RedChannelLowColor, defaultValue=<rgba 0.038 0.0453 0.0625 0>}.RGB) {name=RedChannelLowColor, nodeColor=<rgba 1 0 0 1>}
  NamedRerouteDeclaration(None=VectorParameter() {group=RumbleBeat_TechArt, parameterName=RedChannelMidColor, defaultValue=<rgba 0.957 0.6904 2 1>}.RGB) {name=RedChannelMidColor, nodeColor=<rgba 1 0 0 1>}
  n19 = NamedRerouteUsage()
  n20 = NamedRerouteUsage()
  NamedRerouteDeclaration(None=VertexInterpolator(VS=MaterialFunctionCall(TimestampA=@n19, TransitionLength_In=@n20) {materialFunction=&MF_Timestamp_InOut_01}.Linear).PS) {name=PannerTimeA, desc=ramp up, nodeColor=<rgba 0.0274 0.2424 1 1>}
  NamedRerouteDeclaration(None=VectorParameter() {group=RumbleBeat_TechArt, parameterName=RedChannelHighColor, defaultValue=<rgba 0 0.3325 0.402 1>}.RGB) {name=RedChannelHighColor, nodeColor=<rgba 1 0 0 1>}
  n21 = VectorParameter() {group=RumbleBeat_TechArt, parameterName=GreenChannelLowColor, defaultValue=<rgba 0.4779 0.5423 0.6957 0>}
  NamedRerouteDeclaration(None=@n21.RGB) {name=GreenChannelLowColor, nodeColor=<rgba 0 1 0 1>}
  NamedRerouteDeclaration(None=VectorParameter() {group=RumbleBeat_TechArt, parameterName=GreenChannelMidColor, defaultValue=<rgba 0 0.3325 0.402 1>}.RGB) {name=GreenChannelMidColor, nodeColor=<rgba 0 1 0 1>}
  n22 = NamedRerouteUsage()
  n23 = NamedRerouteUsage()
  n24 = Add(A=TextureCoordinate(), B=AppendVector(A=Multiply(A=@n22) {constB=0.3}, B=Multiply(A=@n23) {constB=0.3}))
  n25 = NamedRerouteUsage()
  n26 = NamedRerouteUsage()
  n27 = NamedRerouteUsage()
  n28 = NamedRerouteUsage()
  n29 = TextureObjectParameter() {
   parameterName=PackedNoise
   samplerType=SAMPLERTYPE_LinearColor
   texture=&RumbleBeatGlow_Noise
  }
  NamedRerouteDeclaration(None=TextureSample(
    UVs=MaterialFunctionCall(UVs=@n24, XTileMultiplier=@n25, YTileMultiplier=@n26, XSpeedMultiplier=@n27, YSpeedMultiplier=@n28) {materialFunction=&MF_AnimateUV}.AnimatedUV
    Tex=@n29
   ).B) {name=BlueChannelTexture, nodeColor=<rgba 0 0 1 1>}
  NamedRerouteDeclaration(None=VectorParameter() {group=RumbleBeat_TechArt, parameterName=GreenChannelHighColor, defaultValue=<rgba 0.8854 0.9498 2 1>}.RGB) {name=GreenChannelHighColor, nodeColor=<rgba 0 1 0 1>}
  NamedRerouteDeclaration(None=VectorParameter() {group=RumbleBeat_TechArt, parameterName=BlueChannelLowColor, defaultValue=<rgba 0.0567 0.087 0.0846 0>}.RGB) {name=BlueChannelLowColor, nodeColor=<rgba 0 0 1 1>}
  NamedRerouteDeclaration(None=VectorParameter() {group=RumbleBeat_TechArt, parameterName=BlueChannelMidColor, defaultValue=<rgba 0 0.3325 0.402 1>}.RGB) {name=BlueChannelMidColor, nodeColor=<rgba 0 0 1 1>}
  NamedRerouteDeclaration(None=VectorParameter() {group=RumbleBeat_TechArt, parameterName=BlueChannelHighColor, defaultValue=<rgba 0.6783 1.4828 2 0>}.RGB) {name=BlueChannelHighColor, nodeColor=<rgba 0 0 1 1>}
  NamedRerouteDeclaration(None=MaterialFunctionCall(A=NamedRerouteUsage(), B=NamedRerouteUsage(), C=NamedRerouteUsage(), Alpha=NamedRerouteUsage()) {materialFunction=&Lerp_3Color}.Result) {name=BlueChannelColor, nodeColor=<rgba 0 0 1 1>}
  NamedRerouteDeclaration(None=MaterialFunctionCall(A=NamedRerouteUsage(), B=NamedRerouteUsage(), C=NamedRerouteUsage(), Alpha=NamedRerouteUsage()) {materialFunction=&Lerp_3Color}.Result) {name=RedChannelColor, nodeColor=<rgba 1 0 0 1>}
  NamedRerouteDeclaration(None=MaterialFunctionCall(A=NamedRerouteUsage(), B=NamedRerouteUsage(), C=NamedRerouteUsage(), Alpha=NamedRerouteUsage()) {materialFunction=&Lerp_3Color}.Result) {name=GreenChannelColor, nodeColor=<rgba 0.4271 1 0.4271 1>}
  NamedRerouteDeclaration(None=@n1.EmissiveColor) {name=In_EmissiveColor, nodeColor=<rgba 0 0.4844 1 1>}
  n30 = Add(
   A=LocalPosition() {localOrigin=InstancePreSkinning}.XYZ
   B=VectorParameter() {group=RumbleBeat_TechArt, parameterName=RadialCoordOffsetCore, defaultValue=<rgba 0 10.9705 2.7421 0>}.RGB
  )
  n31 = ComponentMask(None=VertexInterpolator(VS=@n30).PS) {b=true, g=true}
  n32 = TextureObjectParameter() {
   group=RumbleBeat_TechArt
   parameterName=Core
   samplerType=SAMPLERTYPE_LinearGrayscale
   texture=/BRCosmetics/Weapons/FORT_Melee/Pickaxe_RumbleBeat/Textures/T_RumbleBeat_RadialCore
  }
  n33 = TextureSample(UVs=NamedRerouteDeclaration(None=Multiply(A=@n31) {constB=0.042}) {name=CorePosition, nodeColor=<rgba 0 1 0.5078 1>}, Tex=@n32) {samplerType=SAMPLERTYPE_LinearGrayscale, texture=&T_RumbleBeat_RadialCore}
  NamedRerouteDeclaration(None=@n33.R) {name=CenterTexture, nodeColor=<rgba 1 0.1172 0 1>}
  n34 = NamedRerouteUsage()
  n35 = NamedRerouteUsage()
  n36 = NamedRerouteUsage()
  n37 = NamedRerouteUsage()
  n38 = ComponentMask(None=NamedRerouteUsage()) {r=true}
  n39 = Add(A=NamedRerouteUsage(), B=OneMinus(None=NamedRerouteUsage()))
  NamedRerouteDeclaration(None=LinearInterpolate(A=Multiply(A=@n34) {constB=0.1}, B=LinearInterpolate(A=@n35, B=@n36, Alpha=@n37), Alpha=Multiply(A=@n38, B=@n39))) {name=Out_EmissiveColor, nodeColor=<rgba 0.3828 0 1 1>}
  NamedRerouteDeclaration(None=ScalarParameter() {defaultValue=1, group=RumbleBeat_TechArt, parameterName=RedChannel_UVXScale}) {name=RedChannel_UVXScale, nodeColor=<rgba 1 0 0.0008 1>}
  n40 = VertexInterpolator(VS=Add(
    A=LocalPosition() {localOrigin=InstancePreSkinning}.XYZ
    B=VectorParameter() {group=RumbleBeat_TechArt, parameterName=RadialCoordOffset, defaultValue=<rgba 0 0 -29.5492 0>}.RGB
   ))
  n41 = MaterialFunctionCall(Vector or UVs=Multiply(A=ComponentMask(None=@n40.PS) {b=true, g=true}) {constB=0.0067}) {materialFunction=&VectorToRadialValue}
  n42 = NamedRerouteUsage()
  n43 = NamedRerouteUsage()
  n44 = NamedRerouteUsage()
  n45 = NamedRerouteUsage()
  n46 = TextureObjectParameter() {
   parameterName=PackedNoise
   samplerType=SAMPLERTYPE_LinearColor
   texture=&RumbleBeatGlow_Noise
  }
  n47 = TextureCoordinate()
  NamedRerouteDeclaration(None=TextureSample!(
    UVs=MaterialFunctionCall(UVs=@n41.Radial Coordinates, XTileMultiplier=@n42, YTileMultiplier=@n43, XSpeedMultiplier=@n44, YSpeedMultiplier=@n45) {materialFunction=&MF_AnimateUV}.AnimatedUV
    Tex=@n46
    "DDX(UVs)"=DDX(Value=@n47)
    "DDY(UVs)"=DDY(Value=@n47)
    Apply View MipBias
   ) {mipValueMode=TMVM_Derivative}.R) {name=RedChannelTextureRadial, nodeColor=<rgba 1 0 0 1>}
  n48 = NamedRerouteUsage()
  n49 = NamedRerouteUsage()
  n50 = NamedRerouteUsage()
  n51 = NamedRerouteUsage()
  n52 = TextureObjectParameter() {
   parameterName=PackedNoise
   samplerType=SAMPLERTYPE_LinearColor
   texture=&RumbleBeatGlow_Noise
  }
  n53 = TextureCoordinate()
  NamedRerouteDeclaration(None=TextureSample(
    UVs=MaterialFunctionCall(UVs=@n41.Radial Coordinates, XTileMultiplier=@n48, YTileMultiplier=@n49, XSpeedMultiplier=@n50, YSpeedMultiplier=@n51) {materialFunction=&MF_AnimateUV}.AnimatedUV
    Tex=@n52
    "DDX(UVs)"=DDX(Value=@n53)
    "DDY(UVs)"=DDY(Value=@n53)
   ) {mipValueMode=TMVM_Derivative}.B) {name=BlueChannelTextureRadial, nodeColor=<rgba 0 0 1 1>}
  NamedRerouteDeclaration(None=ScalarParameter() {defaultValue=1, group=RumbleBeat_TechArt, parameterName=RedChannel_UVXScaleRadial, sortPriority=36}) {name=RedChannel_UVXScaleRadial, nodeColor=<rgba 0 1 0.0008 1>}
  NamedRerouteDeclaration(None=ScalarParameter() {defaultValue=1, group=RumbleBeat_TechArt, parameterName=RedChannel_UVYScaleRadial, sortPriority=37}) {name=RedChannel_UVYScaleRadial, nodeColor=<rgba 0 1 0.0008 1>}
  NamedRerouteDeclaration(None=ScalarParameter() {defaultValue=0.056, group=RumbleBeat_TechArt, parameterName=RedChannel_UVXSpeedRadial, sortPriority=38}) {name=RedChannel_UVXSpeedRadial, nodeColor=<rgba 0 1 0.0008 1>}
  NamedRerouteDeclaration(None=ScalarParameter() {defaultValue=-0.41, group=RumbleBeat_TechArt, parameterName=RedChannel__UVYSpeedRadial, sortPriority=39}) {name=RedChannel_UVYSpeedRadial, nodeColor=<rgba 0 1 0.0008 1>}
  NamedRerouteDeclaration(None=ScalarParameter() {defaultValue=1, group=RumbleBeat_TechArt, parameterName=BlueChannel_UVXScaleRadial, sortPriority=40}) {name=BlueChannel_UVXScaleRadial, nodeColor=<rgba 0 0 1 1>}
  NamedRerouteDeclaration(None=ScalarParameter() {defaultValue=2, group=RumbleBeat_TechArt, parameterName=BlueChannel_UVYScaleRadial, sortPriority=41}) {name=BlueChannel_UVYScaleRadial, nodeColor=<rgba 0 0 1 1>}
  NamedRerouteDeclaration(None=ScalarParameter() {defaultValue=0.026, group=RumbleBeat_TechArt, parameterName=BlueChannel_UVXSpeedRadial, sortPriority=42}) {name=BlueChannel_UVXSpeedRadial, nodeColor=<rgba 0 0 1 1>}
  NamedRerouteDeclaration(None=ScalarParameter() {defaultValue=1, group=RumbleBeat_TechArt, parameterName=RedChannel_UVYScale, sortPriority=33}) {name=RedChannel_UVYScale, nodeColor=<rgba 1 0 0.0008 1>}
  NamedRerouteDeclaration(None=ScalarParameter() {defaultValue=-0.614, group=RumbleBeat_TechArt, parameterName=BlueChannel__UVYSpeedRadial, sortPriority=43}) {name=BlueChannel_UVYSpeedRadial, nodeColor=<rgba 0 0 1 1>}
  NamedRerouteDeclaration(None=MaterialFunctionCall(A=NamedRerouteUsage(), B=NamedRerouteUsage(), C=NamedRerouteUsage(), Alpha=NamedRerouteUsage()) {materialFunction=&Lerp_3Color}.Result) {name=BlueChannelColorRadial, nodeColor=<rgba 0 0 1 1>}
  NamedRerouteDeclaration(None=MaterialFunctionCall(A=NamedRerouteUsage(), B=NamedRerouteUsage(), C=NamedRerouteUsage(), Alpha=NamedRerouteUsage()) {materialFunction=&Lerp_3Color}.Result) {name=RedChannelColorRadial, nodeColor=<rgba 1 0 0 1>}
  n54 = NamedRerouteUsage()
  n55 = NamedRerouteUsage()
  n56 = NamedRerouteUsage()
  n57 = NamedRerouteUsage()
  NamedRerouteDeclaration(None=Multiply(A=MaterialFunctionCall(A=@n54, B=@n55, C=@n56, Alpha=@n57) {materialFunction=&Lerp_3Color}.Result) {constB=2}) {name=GreenChannelColorRadial, nodeColor=<rgba 0.4271 1 0.4271 1>}
  n58 = NamedRerouteUsage()
  n59 = NamedRerouteUsage()
  n60 = LocalPosition() {localOrigin=InstancePreSkinning}
  n61 = Multiply(A=ComponentMask(None=VertexInterpolator(VS=@n60.XYZ).PS) {b=true, g=true}) {constB=0.04}
  n62 = ComponentMask(None=VectorParameter() {group=RumbleBeat_TechArt, parameterName=RadialColorsSphereMaskCenter, defaultValue=<rgba 0 0 1.2055 0>}.RGB) {b=true, g=true}
  n63 = ScalarParameter() {defaultValue=0.248, parameterName=Radius}
  n64 = ScalarParameter() {defaultValue=0.128, parameterName=Hardness}
  NamedRerouteDeclaration(None=LinearInterpolate(A=Multiply(A=@n58, B=@n59), B=NamedRerouteUsage(), Alpha=SphereMask(A=@n61, B=@n62, Radius=@n63, Hardness=@n64) {attenuationRadius=12})) {name=AnimatedColorsRadial, nodeColor=<rgba 1 0.7266 0 1>}
  n65 = VertexInterpolator(VS=LocalPosition() {localOrigin=InstancePreSkinning}.XYZ)
  NamedRerouteDeclaration(None=MaterialFunctionCall!(
    A=Multiply(A=@n65.PS) {constB=0.04}
    B=VectorParameter() {parameterName=BoxMaskOffset, defaultValue=<rgba 0 0 1.3979 0>}.RGB
    Bounds=VectorParameter() {parameterName=BoundsBoxMask, defaultValue=<rgba 0.552 1 1 0>}.RGB
    Edge Falloff=ScalarParameter() {defaultValue=0.2, parameterName=EdgeFalloffBoxMask}
   ) {materialFunction=/Engine/Functions/Engine_MaterialFunctions01/Gradient/BoxMask-3D}.Result) {name=BoxMask, nodeColor=<rgba 0 1 0.2031 1>}
  NamedRerouteDeclaration(None=ScalarParameter() {defaultValue=0.1, group=RumbleBeat_TechArt, parameterName=RedChannel_UVXSpeed, sortPriority=34}) {name=RedChannel_UVXSpeed, nodeColor=<rgba 1 0 0.0008 1>}
  NamedRerouteDeclaration(None=ScalarParameter() {group=RumbleBeat_TechArt, parameterName=RedChannel__UVYSpeed, sortPriority=35}) {name=RedChannel_UVYSpeed, nodeColor=<rgba 1 0 0.0008 1>}
  NamedRerouteDeclaration(None=ScalarParameter() {defaultValue=1, group=RumbleBeat_TechArt, parameterName=GreenChannel_UVXScale, sortPriority=36}) {name=GreenChannel_UVXScale, nodeColor=<rgba 0 1 0.0008 1>}
  NamedRerouteDeclaration(None=ScalarParameter() {defaultValue=1, group=RumbleBeat_TechArt, parameterName=GreenChannel_UVYScale, sortPriority=37}) {name=GreenChannel_UVYScale, nodeColor=<rgba 0 1 0.0008 1>}
  NamedRerouteUsage()
  NamedRerouteUsage()
  NamedRerouteUsage()
  n66 = Multiply() {constB=0.1}
  n67 = Multiply() {constB=0.1}
  Subtract(A=Add(A=@n30, B=AppendVector(A=@n66, B=@n67))) {constB=0.05}
 ]}
