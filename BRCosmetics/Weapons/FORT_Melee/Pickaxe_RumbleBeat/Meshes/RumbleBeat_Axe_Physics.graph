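# Physics asset for the RumbleBeat two-handed pickaxe, setting up a short chain that hangs off the handle for physical motion.
# A large capsule covers the handle, and small capsules cover the chain root and its three links.
# The chain root rotates freely on the handle, and each link is constrained to the one before it.
# The first two link joints each swing mainly on alternate axes without twist, while the last allows moderate swing and slight twist.
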PhysicsAsset /BRCosmetics/Weapons/FORT_Melee/Pickaxe_RumbleBeat/Meshes/RumbleBeat_Axe_Physics
bodies {
 MSTR_chain [{shapeType=Capsule, length=0.505, radius=0.505, shapeName=MSTR_chain_capsule, center=<-0.864 0 0>, rotation=<rot 0 -90 0>}]
 Melee_TwoHanded_Handle [{shapeType=Capsule, length=48.5759, radius=4.3732, center=<-0.0225 0.4472 -2.7487>, rotation=<rot 0.037 -90 0.0146>}]
 chain_1 [{shapeType=Capsule, length=0.2, radius=1, shapeName=chain_1_capsule, center=<1.3475 0 0>, rotation=<rot 0 -90 0>}]
 chain_2 [{shapeType=Capsule, length=3, radius=1.3, shapeName=chain_2_capsule, center=<3.5179 0 0>, rotation=<rot 90 0.0004 0>}]
 chain_3 [{shapeType=Capsule, length=3.5, radius=1.5, shapeName=chain_3_capsule, center=<3.7914 0 0>, rotation=<rot 90 0.0009 0>}]
}
constraints [
 {bone1Name=chain_3, bone2Name=chain_2, swing1LimitDegrees=30, swing2LimitDegrees=30, twistLimitDegrees=10}
 {bone1Name=chain_2, bone2Name=chain_1, swing1LimitDegrees=45, swing2LimitDegrees=0, twistLimitDegrees=0}
 {bone1Name=chain_1, bone2Name=MSTR_chain, swing1LimitDegrees=1, swing2LimitDegrees=75, twistLimitDegrees=0}
 {
  bone1Name=MSTR_chain
  bone2Name=Melee_TwoHanded_Handle
  swing1LimitDegrees=45
  swing1Motion=Free
  swing2LimitDegrees=45
  swing2Motion=Free
  twistLimitDegrees=45
  twistMotion=Free
 }
]
