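# Animation Blueprint for the RumbleBeat pickaxe that adds physics motion to its chain bones.
# A RigidBody node simulates the input pose in b_AR_Root space and caches it as Base.
# Two Anim Dynamics chains over chain_1 to chain_3 act as fallback when rigid body is disabled, via RBAN_Off updated each frame.
# bIsInFrontEnd selects the output: world-space simulation in the front end, bone-relative otherwise.
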AnimBlueprint /BRCosmetics/Weapons/FORT_Melee/Pickaxe_RumbleBeat/Meshes/RumbleBeat_Axe_AnimBP
target_skeleton=&/RumbleBeat_Axe_Skeleton
graphs [
 AnimGraph {nodes [
   n1 = Get RBAN_Off() -> (RBAN_Off:bool)
   n2 = Anim Dynamics - Bone: chain_1!(ComponentPose=Local To Component(LocalPose=Use cached pose 'Base'()), bAlphaBoolEnabled:bool=@n1, ExternalForce="0,0,0") -> (Pose) {
    alphaInputType=Bool
    angularDampingOverride=0.9999
    angularSpringConstant=60
    bAngularSpring=true
    bOverrideAngularDamping=true
    bUseGravityOverride=true
    linearDampingOverride=0.99
    linearSpringConstant=550
    simulationSpace=World
    boundBone {boneName=chain_1}
    componentAppliedLinearAccClamp=<3000 3000 6000>
    componentLinearAccScale=<0.4 0.4 0.4>
    componentLinearVelScale=<0.1 0.1 0.1>
    gravityOverride {z=-980}
    physicsBodyDefinitions [{
      boundBone {boneName=chain_1}
      boxExtents=<10 10 10>
      constraintSetup {angularLimitsMax=<0 140 0>, angularLimitsMin=<0 0 0>, angularTarget=<1 0 0>}
      localJointOffset=<0.7 0 0>
     }]
    relativeSpaceBone {boneName=b_AR_Root}
   }
   n3 = Anim Dynamics - Bone: chain_2!(ComponentPose=@n2, bAlphaBoolEnabled:bool=@n1, ExternalForce="0,0,0") -> (Pose) {
    alphaInputType=Bool
    angularDampingOverride=0.9999
    angularSpringConstant=550
    bOverrideAngularDamping=true
    bUseGravityOverride=true
    linearDampingOverride=0.99
    linearSpringConstant=550
    simulationSpace=World
    worldSpaceMinimumScale=0.1
    boundBone {boneName=chain_2}
    componentAppliedLinearAccClamp=<3000 3000 6000>
    componentLinearAccScale=<0.5 0.5 1>
    componentLinearVelScale=<0.2 0.2 0.1>
    gravityOverride {z=-980}
    physicsBodyDefinitions [{
      boundBone {boneName=chain_2}
      boxExtents=<10 10 10>
      constraintSetup {angularLimitsMax=<0 0 40>, angularLimitsMin=<0 0 -40>, angularTarget=<1 0 0>}
      localJointOffset=<0.7 0.1 0.1>
     }]
    relativeSpaceBone {boneName=b_AR_Root}
   }
   n4 = Anim Dynamics - Bone: chain_3!(ComponentPose=@n3, bAlphaBoolEnabled:bool=@n1, ExternalForce="0,0,0") -> (Pose) {
    alphaInputType=Bool
    angularDampingOverride=1
    angularSpringConstant=550
    bAngularSpring=true
    bOverrideAngularDamping=true
    bUseGravityOverride=true
    linearDampingOverride=0.99
    linearSpringConstant=550
    simulationSpace=World
    worldSpaceMinimumScale=0.1
    boundBone {boneName=chain_3}
    componentAppliedLinearAccClamp=<3000 3000 6000>
    componentLinearAccScale=<0.3 0.3 1>
    componentLinearVelScale=<0.2 0.2 0.1>
    gravityOverride {z=-980}
    physicsBodyDefinitions [{
      boundBone {boneName=chain_3}
      boxExtents=<10 10 10>
      constraintSetup {angularLimitsMax=<0 50 50>, angularLimitsMin=<0 -50 -50>, angularTarget=<1 0 0>}
      localJointOffset=<1.5 0.1 0.1>
     }]
    relativeSpaceBone {boneName=b_AR_Root}
   }
   n5 = Get RBAN_Off()
   n6 = Anim Dynamics - Bone: chain_1(ComponentPose=Local To Component(LocalPose=Use cached pose 'Base'()), bAlphaBoolEnabled=@n5, ExternalForce="0,0,0") {
    alphaInputType=Bool
    angularDampingOverride=0.9999
    angularSpringConstant=60
    bAngularSpring=true
    bOverrideAngularDamping=true
    bUseGravityOverride=true
    linearDampingOverride=0.99
    linearSpringConstant=550
    simulationSpace=BoneRelative
    boundBone {boneName=chain_1}
    componentAppliedLinearAccClamp=<3000 3000 6000>
    componentLinearAccScale=<0.4 0.4 0.4>
    componentLinearVelScale=<0.1 0.1 0.1>
    gravityOverride {z=-980}
    physicsBodyDefinitions [{
      boundBone {boneName=chain_1}
      boxExtents=<10 10 10>
      constraintSetup {angularLimitsMax=<0 140 0>, angularLimitsMin=<0 0 0>, angularTarget=<1 0 0>}
      localJointOffset=<1 0 0>
     }]
    relativeSpaceBone {boneName=b_AR_Root}
   }
   n7 = Anim Dynamics - Bone: chain_2(ComponentPose=@n6, bAlphaBoolEnabled=@n5, ExternalForce="0,0,0") {
    alphaInputType=Bool
    angularDampingOverride=0.9999
    angularSpringConstant=550
    bOverrideAngularDamping=true
    bUseGravityOverride=true
    linearDampingOverride=0.99
    linearSpringConstant=550
    simulationSpace=BoneRelative
    worldSpaceMinimumScale=0.1
    boundBone {boneName=chain_2}
    componentAppliedLinearAccClamp=<3000 3000 6000>
    componentLinearAccScale=<0.5 0.5 1>
    componentLinearVelScale=<0.2 0.2 0.1>
    gravityOverride {z=-980}
    physicsBodyDefinitions [{
      boundBone {boneName=chain_2}
      boxExtents=<10 10 10>
      constraintSetup {angularLimitsMax=<0 0 40>, angularLimitsMin=<0 0 -40>, angularTarget=<1 0 0>}
      localJointOffset=<1 0.1 0.1>
     }]
    relativeSpaceBone {boneName=b_AR_Root}
   }
   n8 = Anim Dynamics - Bone: chain_3(ComponentPose=@n7, bAlphaBoolEnabled=@n5, ExternalForce="0,0,0") {
    alphaInputType=Bool
    angularDampingOverride=1
    angularSpringConstant=550
    bAngularSpring=true
    bOverrideAngularDamping=true
    bUseGravityOverride=true
    linearDampingOverride=0.99
    linearSpringConstant=550
    simulationSpace=BoneRelative
    worldSpaceMinimumScale=0.1
    boundBone {boneName=chain_3}
    componentAppliedLinearAccClamp=<3000 3000 6000>
    componentLinearAccScale=<0.3 0.3 1>
    componentLinearVelScale=<0.2 0.2 0.1>
    gravityOverride {z=-980}
    physicsBodyDefinitions [{
      boundBone {boneName=chain_3}
      boxExtents=<10 10 10>
      constraintSetup {angularLimitsMax=<0 50 50>, angularLimitsMin=<0 -50 -50>, angularTarget=<1 0 0>}
      localJointOffset=<2 0.1 0.1>
     }]
    relativeSpaceBone {boneName=b_AR_Root}
   }
   Output Pose(Result=Blend Poses by bool(BlendPose_0=Component To Local(ComponentPose=@n4), BlendPose_1=Component To Local(ComponentPose=@n8), bActiveValue=Get bIsInFrontEnd())) {layerGroup=None}
   AnimGraphNode_SubInput_0 = Input Pose()
   n9 = RigidBody!(ComponentPose=Local To Component(LocalPose=@AnimGraphNode_SubInput_0), bAlphaBoolEnabled:bool, ExternalForce) -> (Pose) {
    alphaInputType=Bool
    bOverrideWorldGravity=true
    overridePhysicsAsset=&/RumbleBeat_Axe_Physics
    simulationSpace=BaseBoneSpace
    baseBoneRef {boneName=b_AR_Root}
    componentAppliedLinearAccClamp=<3000 3000 3000>
    overrideWorldGravity {z=-980}
    simSpaceSettings {dampingAlpha=0, maxAngularAcceleration=9, maxAngularVelocity=3, maxLinearAcceleration=1500, maxLinearVelocity=450, worldAlpha=1}
   }
   Save cached pose 'Base'(Pose=Component To Local(ComponentPose=@n9)) {cacheName=Base}
  ]}
 EventGraph {nodes [
   Event BlueprintUpdateAnimation()
   > Set RBAN_Off(RBAN_Off:bool=NOT Boolean(A=Get bRigidBodyEnabled())) -> (Output_Get:bool)
   TryGetPawnOwner()
  ]}
]
variables [RBAN_Off {type=boolean}]
